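# Translucent unlit material for drifting distant fog, tinted by time of day or scaled by rainforest fog collection parameters.
# Panning, distorted textures with optional parallax and ground-fog layers build opacity, shaped by depth, near, camera-height, edge and world-mask fades.
# Color blends shadowed and lit tints using a normal map against the atmospheric light.
# Quality switches simplify lower tiers, and a height-driven offset collapses vertices when faded out.
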Material /Game/Athena/Environments/Materials/DistantFog/M_DistantFog
graph {
 properties {
  bCastRayTracedShadows=false
  bUseTranslucencyVertexFog=false
  blendMode=BLEND_Translucent
  pixelDepthOffsetMode=PDOM_Legacy
  shadingModel=MSM_Unlit
  usedShadingModels=Unlit
 }
 nodes [
  Composite() {subgraphName=TOD Colors}
  n1 = CameraPositionWS()
  n2 = Divide(
   A=Subtract(A=ScalarParameter() {defaultValue=90000, group=HeightFade, parameterName=HeightFadeInZ}, B=ComponentMask(None=@n1) {b=true}) {constA=90000}
   B=ScalarParameter() {defaultValue=20000, group=HeightFade, parameterName=HeightFadeInDistance}
  ) {constB=20000}
  Max(
   A=Power(Base=Saturate(None=@n2), Exp=ScalarParameter() {defaultValue=1, group=HeightFade, parameterName=HeightFadeInExponent})
   B=ScalarParameter() {defaultValue=0.25, group=HeightFade, parameterName=HeightFadeInMinOpacity}
  ) {desc="originally to increase biome contrast from sky\r\nreplaced by world mask"}
  n3 = CollectionParameter() {
   collection=&FortniteMaterialParameters
   parameterName=RainForest_Fog_DistantOpacity
  }
  n4 = ScalarParameter() {defaultValue=0.25, group=TimeOfDay, parameterName=TODFactor_Morning}
  n5 = ScalarParameter() {defaultValue=1, group=TimeOfDay, parameterName=TODFactor_Day}
  n6 = ScalarParameter() {defaultValue=0.2, group=TimeOfDay, parameterName=TODFactor_Evening}
  n7 = ScalarParameter() {defaultValue=0.125, group=TimeOfDay, parameterName=TODFactor_Night}
  n8 = Saturate(None=DotProduct(A=CameraVectorWS(), B=VertexNormalWS()))
  n9 = ScalarParameter() {defaultValue=2, group=Fresnel, parameterName=FresnelExponent}
  n10 = ScalarParameter() {defaultValue=2000, group=DistantFog, parameterName=DepthFadeDistance}
  n11 = NamedRerouteDeclaration(None=SmoothStep(Value=DepthFade(FadeDistance=@n10))) {name=DepthFade, nodeColor=<rgba 1 0 0.1875 1>}
  n12 = ScalarParameter() {defaultValue=1, parameterName=DetailTilingX}
  n13 = ScalarParameter() {defaultValue=1, parameterName=DetailTilingY}
  n14 = TextureSampleParameter2D(UVs=Multiply(A=TextureCoordinate(), B=AppendVector(A=@n12, B=@n13))) {
   parameterName=FogDetailTexture
   samplerSource=SSM_Wrap_WorldGroupSettings
   sortPriority=0
   texture=/Game/Global/Textures/Default/Blanks/T_MidgreysRGB
  }
  MaterialExpressionRemap_0 = MaterialXRemap!(
   None=@n14.RGB
   InputLow
   InputHigh
   TargetLow=ScalarParameter() {parameterName=DetailLow}
   TargetHigh=ScalarParameter() {defaultValue=1, parameterName=DetailHigh}
  )
  n15 = Multiply(
   A=TextureCoordinate()
   B=AppendVector(A=ScalarParameter() {defaultValue=1, parameterName=MeshUVScaleX}, B=ScalarParameter() {defaultValue=1, parameterName=MeshUVScaleY})
  )
  n16 = ScalarParameter() {parameterName=MeshUVRotation}
  n17 = WorldPosition()
  n18 = ScalarParameter() {defaultValue=20000, group=DistantFog, parameterName=FogUVSize, sortPriority=100}
  n19 = StaticSwitchParameter(
   True=NamedRerouteDeclaration(None=MaterialFunctionCall(UVs=@n15, "Rotation Angle (0-1)"=@n16) {materialFunction=&CustomRotator}.Rotated Values) {name=MeshUVs, nodeColor=<rgba 0 1 0.2266 1>}
   False=ComponentMask(None=Divide(A=@n17.XYZ, B=@n18)) {g=true, r=true}
  ) {parameterName=UseMeshUVs}
  n20 = Add(A=@n19, B=Multiply(A=NamedRerouteUsage(), B=ScalarParameter() {defaultValue=0.5, group=Parallax, parameterName=FogParallaxMult}))
  n21 = Time()
  n22 = ScalarParameter() {defaultValue=0.001, group=DistantFog, parameterName=PanningSpeedX}
  n23 = ScalarParameter() {defaultValue=0.001, group=DistantFog, parameterName=PanningSpeedY}
  n24 = Add(
   A=QualitySwitch(Default=StaticSwitchParameter(True=@n20, False=@n19) {group=Parallax, parameterName=UseParallaxBoost, sortPriority=0}, Low=@n19)
   B=Multiply(A=@n21, B=AppendVector(A=@n22, B=@n23)) {constB=0.003}
  )
  n25 = Divide(A=WorldPosition().XYZ, B=ScalarParameter() {defaultValue=20000, group=DistantFog, parameterName=DistortionUVSize, sortPriority=101})
  n26 = ScalarParameter() {defaultValue=0.005, group=DistantFog, parameterName=DistortionSpeed}
  n27 = Time()
  n28 = TextureSampleParameter2D(UVs=Add(A=ComponentMask(None=@n25) {g=true, r=true}, B=Multiply(A=@n26, B=@n27))) {
   parameterName=DistortionTexture
   samplerSource=SSM_Wrap_WorldGroupSettings
   samplerType=SAMPLERTYPE_LinearGrayscale
   texture=/Game/Global/Textures/Default/Blanks/T_BlackLinearGrayscale
  }
  n29 = NamedRerouteDeclaration(None=ConstantBiasScale(None=ComponentMask(None=@n28.RGB) {g=true, r=true}) {bias=-0.5, scale=2}) {name=BaseDistortion, nodeColor=<rgba 0.7812 1 0 1>}
  n30 = ScalarParameter() {defaultValue=0.25, group=DistantFog, parameterName=DistortionIntensity}
  n31 = QualitySwitch(Default=Add(A=@n24, B=Multiply(A=@n29, B=@n30)), Low=@n24, Medium=@n24)
  n32 = TextureSampleParameter2D(UVs=@n31) {
   parameterName=FogTexture
   samplerSource=SSM_Wrap_WorldGroupSettings
   sortPriority=0
   texture=&T_MidgreysRGB
  }
  n33 = Power(
   Base=NamedRerouteDeclaration(None=@n32.R) {name=FogBaseDensity, nodeColor=<rgba 0 1 0.3672 1>}
   Exp=ScalarParameter() {defaultValue=0.5, group=DistantFog, parameterName=FogOpacityPower}
  )
  n34 = Multiply(A=@MaterialExpressionRemap_0, B=@n33)
  n35 = Multiply(
   A=QualitySwitch(Default=StaticSwitchParameter(True=@n34, False=@n33) {parameterName=UseDetailFogTexture}, Low=@n33, Medium=@n33)
   B=ScalarParameter() {defaultValue=0.8, group=DistantFog, parameterName=FogOpacity}
  )
  n36 = QualitySwitch(
   Default=StaticSwitchParameter(True=Multiply(A=@n11, B=@n35), False=@n35) {defaultValue=true, group=DistantFog, parameterName=UseDistantFogDepthFade}
   Low=@n35
  )
  n37 = Divide(
   A=MaterialFunctionCall() {materialFunction=&WorldPositionBehindTranslucency}.Result
   B=ScalarParameter() {defaultValue=20000, group=GroundFog, parameterName=GroundFogUVSize}
  )
  n38 = MaterialFunctionCall!(
   HeightTexture=ScalarParameter() {group=Parallax, parameterName=ParallaxIntensity, sortPriority=1}
   TextureCoordinates
   ReferencePlane=ScalarParameter() {defaultValue=1, group=Parallax, parameterName=ParallaxReferencePlane}
   HeightRatioInput
   CameraVector=CameraVectorWS()
  ) {materialFunction=/Engine/Functions/Engine_MaterialFunctions02/Utility/BumpOffset_Advanced.BumpOffset_advanced}
  n39 = Divide(A=WorldPosition().XYZ, B=ScalarParameter() {defaultValue=20000, group=GroundFog, parameterName=FogUVGroundSize, sortPriority=100})
  n40 = StaticSwitchParameter(
   True=NamedRerouteUsage()
   False=Add(A=ComponentMask(None=@n39) {g=true, r=true}, B=ScalarParameter() {defaultValue=0.5, parameterName=offset})
  ) {parameterName=UseMeshUVs}
  n41 = Subtract(
   A=Multiply(
    A=NamedRerouteDeclaration(None=@n38.Distortion) {name=ParallaxOffset, nodeColor=<rgba 0 0.5644 1 1>}
    B=ScalarParameter() {defaultValue=1, group=Parallax, parameterName=GroundFogParallaxMult}
   )
   B=@n40
  )
  n42 = Multiply(A=@n29, B=ScalarParameter() {defaultValue=0.25, group=GroundFog, parameterName=GroundDistortionIntensity})
  n43 = Multiply(A=@n21, B=ScalarParameter() {defaultValue=0.003, group=GroundFog, parameterName=GroundPanningSpeed}) {constB=0.003}
  n44 = Subtract(
   A=StaticSwitchParameter(
    True=ComponentMask(None=@n37) {g=true, r=true}
    False=StaticSwitchParameter(True=@n41, False=@n40) {group=Parallax, parameterName=UseParallaxBoost, sortPriority=0}
   ) {parameterName=ProjectGroundFog}
   B=QualitySwitch(Default=Add(A=@n42, B=@n43), Low=@n43, Medium=@n43)
  )
  n45 = TextureSampleParameter2D(UVs=@n44) {
   parameterName=GroundFogTexture
   samplerSource=SSM_Wrap_WorldGroupSettings
   sortPriority=0
   texture=&T_MidgreysRGB
  }
  n46 = TextureSampleParameter2D(UVs=@n44) {
   parameterName=FogTexture
   samplerSource=SSM_Wrap_WorldGroupSettings
   sortPriority=0
   texture=&T_MidgreysRGB
  }
  n47 = StaticSwitchParameter(True=@n45.R, False=@n46.R) {group=GroundFog, parameterName=UseDedicatedGroundFogTexture, sortPriority=0}
  n48 = ScalarParameter() {defaultValue=1, group=Groundfog, parameterName=GroundFogOpacity}
  n49 = ScalarParameter() {defaultValue=4, group=Scalability, parameterName=MED_GroundFogOpacityMult}
  n50 = Multiply(
   A=Power(
    Base=QualitySwitch(Default=@n47, Low=@n46.R, Medium=@n46.R)
    Exp=ScalarParameter() {defaultValue=0.25, group=GroundFog, parameterName=GroundFogPower}
   )
   B=QualitySwitch(Default=@n48, Medium=Multiply(A=@n48, B=@n49))
  )
  n51 = StaticSwitchParameter(True=Multiply(A=NamedRerouteUsage(), B=@n50), False=@n50) {group=GroundFog, parameterName=UseGroundFogDepthFade}
  n52 = QualitySwitch(
   Default=StaticSwitchParameter(True=Add(A=@n36, B=@n51), False=@n36) {defaultValue=true, group=GroundFog, parameterName=EnableGroundFog, sortPriority=0}
   Low=@n36
  )
  n53 = TextureCoordinate()
  n54 = Multiply(A=NamedRerouteUsage(), B=ScalarParameter() {defaultValue=0.015, group=SphereMask, parameterName=SphereMaskDistortionIntensity})
  n55 = SphereMask(
   A=QualitySwitch(Default=Add(A=@n53, B=@n54), Low=@n53, Medium=@n53)
   B=Constant() {r=0.5}
   Radius=ScalarParameter() {defaultValue=0.49, group=SphereMask, parameterName=SphereMaskRadius}
   Hardness=ScalarParameter() {defaultValue=0.5, group=SphereMask, parameterName=SphereMaskHardness}
  ) {attenuationRadius=0.5, hardnessPercent=50}
  n56 = StaticSwitchParameter(True=Multiply(A=@n52, B=SmoothStep(Value=@n55)), False=@n52) {group=SphereMask, parameterName=UseSphereMask, sortPriority=0}
  n57 = StaticSwitchParameter(True=Multiply(A=Power(Base=@n8, Exp=@n9), B=@n56), False=@n56) {group=Fresnel, parameterName=UseFresnel}
  n58 = Add(A=@n57, B=NamedRerouteUsage())
  n59 = ComponentMask(None=CameraPositionWS()) {b=true}
  n60 = Add(
   A=ComponentMask(None=ObjectPositionWS()) {b=true}
   B=ScalarParameter() {defaultValue=-2000, group=HeightFade, parameterName=HeightFadeCullOffset}
  )
  n61 = ScalarParameter() {defaultValue=8000, group=HeightFade, parameterName=HeightFadeDistance}
  n62 = Saturate(None=Divide(A=Subtract(A=@n59, B=@n60), B=@n61))
  n63 = Add(
   A=ComponentMask(None=CameraPositionWS()) {b=true}
   B=ScalarParameter() {defaultValue=-2000, group=HeightFade, parameterName=HeightFadeCullOffset}
  )
  n64 = Add(A=ScalarParameter() {group=HeightFade, parameterName=HeightFadeOffset}, B=@n63)
  n65 = ComponentMask(None=ObjectPositionWS()) {b=true}
  n66 = Add(A=@n61, B=@n65)
  n67 = StaticSwitchParameter(True=@n62, False=Saturate(None=Divide(A=@n64, B=@n66) {constB=10000})) {parameterName=UseSimpleHeightFade}
  n68 = ScalarParameter() {defaultValue=4, group=HeightFade, parameterName=HeightFadePower}
  n69 = Multiply(
   A=StaticSwitchParameter(
    True=@n3
    False=MaterialFunctionCall(Value_Morning=@n4, Value_Day=@n5, Value_Evening=@n6, Value_Night=@n7) {materialFunction=&Scalar_Time_of_Day}.Result
   ) {desc=FortniteMaterialParameters MPC - RainForest_Fog_DistantOpacity, group=TimeOfDay, parameterName=UseOpacityDrivenByMPC}
   B=Multiply(
    A=StaticSwitchParameter(True=@n58, False=@n57) {group=GroundFog, parameterName=UseFakeGroundFog}
    B=Power(Base=@n67, Exp=@n68) {constExponent=4}
   )
  )
  n70 = ScalarParameter() {defaultValue=1.5, group=Scalability, parameterName=LOW_OpacityMult_MorningEvening}
  n71 = Multiply(
   A=@n69
   B=MaterialFunctionCall(
    Value_Morning=@n70
    Value_Day=ScalarParameter() {defaultValue=1.5, group=Scalability, parameterName=LOW_OpacityMult_Day}
    Value_Evening=@n70
    Value_Night=ScalarParameter() {defaultValue=3, group=Scalability, parameterName=LOW_OpacityMult_Night}
   ) {materialFunction=&Scalar_Time_of_Day}.Result
  )
  n72 = CameraPositionWS()
  n73 = WorldPosition()
  n74 = ScalarParameter() {defaultValue=10000, group=DistantFog, parameterName=NearFadeDistance}
  n75 = Multiply(A=QualitySwitch(Default=@n69, Low=@n71, Medium=@n71), B=OneMinus(None=SphereMask(A=@n72, B=@n73.XYZ, Radius=@n74) {hardnessPercent=0}))
  n76 = VertexColor()
  n77 = WorldPosition() {desc=based on RVT volume}
  n78 = ScalarParameter() {defaultValue=130048, group=WorldMask, parameterName=WorldCoordinatesOffset}
  n79 = TextureSampleParameter2D(UVs=Divide(A=Add(A=@n77.XY, B=@n78) {constB=130048}, B=ScalarParameter() {defaultValue=260096, group=WorldMask, parameterName=WorldCoordinatesSize}) {constB=260096}) {
   group=WorldMask
   parameterName=WorldMask
   samplerSource=SSM_Clamp_WorldGroupSettings
   texture=&T_BlacksRGB
  }
  n80 = StaticSwitchParameter(True=@n76, False=QualitySwitch(Default=@n79.RGB, Low=@n76, Medium=@n76)) {group=WorldMask, parameterName=ForceVertexColorMask}
  n81 = StaticSwitchParameter(
   True=Multiply(A=@n75, B=StaticComponentMaskParameter(None=@n80) {defaultR=true, group=WorldMask, parameterName=WorldMaskChannel})
   False=@n75
  ) {desc="VertexColor is used on LOW && MEDIUM", group=WorldMask, parameterName=UseWorldTextureMask}
  n82 = Subtract(A=TextureCoordinate()) {constB=0.5}
  n83 = OneMinus(None=Multiply(A=Abs(None=@n82)) {constB=2})
  n84 = ComponentMask(None=@n83) {g=true}
  n85 = CollectionParameter() {
   collection=&FortniteMaterialParameters
   parameterName=RainForest_Fog_MasterOpacity
  }
  Multiply(A=StaticSwitchParameter(True=Multiply(A=@n81, B=@n84), False=@n81) {group=EdgeFade, parameterName=UseEdgeFade}, B=@n85)
  Multiply(
   A=Power(Base=@n84, Exp=ScalarParameter() {defaultValue=1, parameterName=EdgeFade_PowerX}) {constExponent=1}
   B=Power(Base=ComponentMask(None=@n83) {g=true}, Exp=ScalarParameter() {defaultValue=1, parameterName=EdgeFade_PowerY}) {constExponent=1}
  )
  NamedRerouteDeclaration(None=@n44) {name=GroundUVs, nodeColor=<rgba 1 0 0.3984 1>}
  NamedRerouteDeclaration(None=@n31) {name=FogUVs, nodeColor=<rgba 0.9688 0 1 1>}
  n86 = ComponentMask(None=ObjectPositionWS()) {g=true, r=true}
  n87 = MaterialFunctionCall() {materialFunction=&WorldPositionBehindTranslucency}
  n88 = ComponentMask(None=@n87.Result) {g=true, r=true}
  n89 = ScalarParameter() {defaultValue=5000, group=Groundfog, parameterName=GroundFogRadius}
  n90 = ScalarParameter() {defaultValue=0.5, group=Groundfog, parameterName=GroundFogHardness}
  n91 = Subtract(A=ComponentMask(None=@n87.Result) {b=true}, B=ScalarParameter() {group=Groundfog, parameterName=GroundFogBias})
  n92 = ScalarParameter() {defaultValue=5000, group=Groundfog, parameterName=GroundFogDistance}
  n93 = OneMinus(None=Saturate(None=Divide(A=@n91, B=@n92)))
  n94 = ScalarParameter() {defaultValue=1, group=Groundfog, parameterName=GroundFogOpacity}
  NamedRerouteDeclaration(None=Multiply(A=SphereMask(A=@n86, B=@n88, Radius=@n89, Hardness=@n90), B=Multiply(A=@n93, B=@n94))) {name=GroundFog, nodeColor=<rgba 0.3359 0 1 1>}
  PinBase()
  n95 = MaterialFunctionCall(
   Color_Morning=VectorParameter() {group=TimeOfDay, parameterName=FogTint_Morning, defaultValue=<rgba 0.401 0.2322 0.2659 1>}.RGB
   Color_Day=VectorParameter() {group=TimeOfDay, parameterName=FogTint_Day, defaultValue=<rgba 0.276 0.276 0.276 1>}.RGB
   Color_Evening=VectorParameter() {group=TimeOfDay, parameterName=FogTint_Evening, defaultValue=<rgba 0.2183 0.2376 0.241 1>}.RGB
   Color_Night=VectorParameter() {group=TimeOfDay, parameterName=FogTint_Night, defaultValue=<rgba 0.3188 0.4123 0.693 1>}.RGB
  ) {materialFunction=&Color_Time_of_Day}
  PinBase!(Result=@n95.Result) {reroutePins [Result {expression=&self:MaterialExpressionReroute_2}]}
  n96 = WorldPosition()
  n97 = OneMinus(None=Ceil(None=@n62))
  n98 = Divide(A=@n63, B=Abs(None=@n65)) {constB=10000}
  n99 = OneMinus(None=Floor(None=Saturate(None=@n98)))
  Reroute(None=Multiply(A=Multiply(A=@n96.XYZ) {constB=-1}, B=StaticSwitchParameter(True=@n97, False=@n99) {parameterName=UseSimpleHeightFade}))
  MaterialExpressionReroute_2 = Reroute(None=@n95.Result)
  n100 = Desaturation(
   None=@MaterialExpressionReroute_2.Result
   Fraction=ScalarParameter() {defaultValue=0.5, group=TimeOfDay, parameterName=ShadowedColor_TODInfluence}
  )
  n101 = VectorParameter() {parameterName=ShadowedColor, defaultValue=<rgba 0.1067 0.2505 0.5608 1>}
  n102 = TextureSampleParameter2D(UVs=StaticSwitchParameter(True=NamedRerouteUsage(), False=NamedRerouteUsage()) {parameterName=UseGroundUVsForNormal}) {parameterName=FogNormal, samplerType=SAMPLERTYPE_Normal, texture=&FlatNormalBlue}
  n103 = Multiply(A=ScalarParameter() {defaultValue=1, group=DistantFog, parameterName=ShadingBoost}) {constB=-1}
  n104 = Saturate(None=DotProduct(
    A=MaterialFunctionCall(Normal=@n102.RGB, Flatness=@n103) {materialFunction=&FlattenNormal}.Result
    B=AtmosphericLightVector()
   ))
  n105 = MaterialFunctionCall(
   In=Power(Base=NamedRerouteUsage(), Exp=ScalarParameter() {defaultValue=0.5, group=Scalability, parameterName=LOW_ColorGradientPower}) {constExponent=0.5}
   Contrast=ScalarParameter() {defaultValue=1, group=Scalability, parameterName=LOW_ColorGradientContrast}
  ) {materialFunction=&CheapContrast}
  n106 = Multiply(
   A=LinearInterpolate(
    A=Multiply(A=@n100, B=@n101.RGB)
    B=@MaterialExpressionReroute_2.Result
    Alpha=QualitySwitch(Default=@n104, Low=@n105.Result, Medium=@n105.Result)
   )
   B=VectorParameter() {group=DistantFog, parameterName=FogTint, sortPriority=1, defaultValue=<rgba 1 1 1 1>}.RGB
  )
  ShadingPathSwitch(Default=@n106, Mobile=Multiply(A=@n106, B=ScalarParameter() {defaultValue=0.75, parameterName=FogTintMobile/PerfScale}))
 ]
}
